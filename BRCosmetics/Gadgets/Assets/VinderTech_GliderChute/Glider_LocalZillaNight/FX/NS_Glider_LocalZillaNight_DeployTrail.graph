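# Niagara deploy trail effect for the LocalZilla Night glider, drawing short-lived ribbons from bones or sockets on a skeletal mesh.
# A single emitter uses the AnimTrail modules to spawn and update particles for two ribbon renderers.
# User.Color is dimmed by time of day for evening and night.
# Alpha and width curves combine particle age with emitter loop age, which also drives a dynamic material parameter over the deploy.
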NiagaraSystem /BRCosmetics/Gadgets/Assets/VinderTech_GliderChute/Glider_LocalZillaNight/FX/NS_Glider_LocalZillaNight_DeployTrail
dependencies {usedDataInterfaces [&NiagaraDataInterfaceCurve], usedDynamicInputs [], usedModules [&SystemState]}
emitters {Lightning_01001 {
  emitterUpdateScript [
   &EmitterState()
   &FN_SpawnAnimTrail_v2()
  ]
  particleSpawnScript [
   /Game/Effects/Niagara/Modules/AnimTrail/FN_InitializeAnimTrail/FN_InitializeAnimTrailBoneIndecies(Second Bone or Socket Index=1)
   /Game/Effects/Niagara/Modules/AnimTrail/FN_InitializeAnimTrail/FN_InitializeAnimTrail_v3/FN_InitializeAnimTrail(
    AnimTrailWidth=3.5
    Color:LinearColor=&MultiplyLinearColor(
     Linear Color=$User.Color
     Scale Factor=&MakeFloat4FromVectorAndFloat(
      "Vector (RGB)"=&VectorFromFloat(Value=&FN_Select_Float_from_Time_Of_Day(Morning=0.8, Day=1, Evening=0.15, Night=0.15))
      "Float (Alpha)"=1
     )
    )
    Lifetime=0.4
   ) {module=FN_InitializeAnimTrail001}
   &AddVelocity(Velocity=<0 1000 0>)
  ]
  particleUpdateScript [
   &SolveForcesAndVelocity(Clamp Velocity=false, Limit Acceleration=false)
   /Game/Effects/Niagara/Modules/AnimTrail/FN_UpdateAnimTrail/FN_UpdateAnimTrail_v3/FN_UpdateAnimTrail(Locked Particle RibbonWidth:NiagaraFloat=$Particles.RibbonWidth)
   &ScaleColor(
    Scale Alpha=&Multiply_Float(
     A=&Multiply_Float(
      A=&FloatFromCurve(
       FloatCurve=DataInterface() {propertyValues {Curve {keys [
           {arriveTangent=-0.102, interpMode=RCIM_Cubic, leaveTangent=-0.102, tangentMode=RCTM_User, time=0.5, value=1}
           {arriveTangent=-0.4651, interpMode=RCIM_Cubic, leaveTangent=-0.4651, tangentMode=RCTM_User, time=1, value=0}
          ]}}}
       Scale Curve=0.45
      )
      B=1
     )
     B=&FloatFromCurve(
      FloatCurve=DataInterface() {propertyValues {Curve {keys [
          <at 0 0 RCIM_Cubic>
          {arriveTangent=0.2254, interpMode=RCIM_Cubic, leaveTangent=0.2254, tangentMode=RCTM_User, time=0.35, value=1}
          {interpMode=RCIM_Cubic, tangentMode=RCTM_User, time=0.8, value=1}
          {interpMode=RCIM_Cubic, tangentMode=RCTM_User, time=1, value=0}
         ]}}}
      CurveIndex=$Emitter.NormalizedLoopAge
     )
    )
    Scale RGB=&VectorFromFloat(Value=1)
    ScaleA=true
    ScaleRGB=true
   )
   &ScaleRibbonWidth(Ribbon Width Scale=&Multiply_Float(
     A=&FloatFromCurve(FloatCurve=DataInterface() {propertyValues {Curve {keys [{interpMode=RCIM_Cubic, tangentMode=RCTM_User, time=0, value=1}, {interpMode=RCIM_Cubic, tangentMode=RCTM_User, time=1, value=0.5}]}}})
     B=&FloatFromCurve(
      FloatCurve=DataInterface() {propertyValues {Curve {keys [
          {interpMode=RCIM_Cubic, tangentMode=RCTM_User, time=0.2, value=0}
          {arriveTangent=0.625, interpMode=RCIM_Cubic, leaveTangent=0.625, time=0.35, value=1}
          <at 1 0.5 RCIM_Cubic>
         ]}}}
      CurveIndex=$Emitter.NormalizedLoopAge
     )
    ))
   &FN_Sprite_Attributes(
    Color=&MultiplyLinearColor(Linear Color=$Particles.Color, Scale Factor=<1 1 1 0.9>)
    Create Color Attribute=true
    Write to Module Namespace=true
   )
   &FN_Sprite_Attributes(
    Color=&MultiplyLinearColor(Linear Color=$Particles.Color, Scale Factor=<1 1 1 1>)
    Create Color Attribute=true
    Write to Module Namespace=true
   ) {module=FN_Sprite_Attributes001}
   &DynamicMaterialParameters(
    Index 0 Param 1=1
    Index 0 Param 3=1
    Index 0 Param 4=&Lerp_Float(A=0, B=-2.2, Alpha=$Emitter.NormalizedLoopAge)
    Module.Param0WriteEnabled=true
    Module.Param1WriteEnabled=true
    Module.Param2WriteEnabled=true
    Module.Param3WriteEnabled=true
   )
  ]
  rendererClasses [&NiagaraRibbonRendererProperties]
  renderers [
   {rendererClass=&NiagaraRibbonRendererProperties}
   {rendererClass=&NiagaraRibbonRendererProperties, rendererIndex=1}
  ]
 }}
summary {
 systemName=NS_Glider_LocalZillaNight_DeployTrail
 emitters [{emitterName=Lightning_01001, rendererClasses [&NiagaraRibbonRendererProperties]}]
 userVariables [
  User.Color {type=LinearColor, description="", defaultValue {struct=LinearColor, value=<rgba 0.0521 0.0521 0.0521 1>}}
  User.RibbonIDOffset {type=NiagaraInt32}
  User.SkeletalMesh {
   type=NiagaraDataInterfaceSkeletalMesh
   defaultValue {
    struct=NiagaraExt_VariableValue_DataInterface
    value {dataInterface=&self:NiagaraDataInterfaceSkeletalMesh_0}
   }
  }
 ]
}
